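# One-shot mono MetaSound Source that plays a stone-pull grip sound for VanillaSoda.
# The OnPlay trigger starts a single Wave Player on the exposed Sound input, which defaults to one grip wave, with no randomization or processing.
# The player feeds the mono output directly, and its On Finished trigger ends the one-shot.
# It uses logarithmic distance attenuation and the melee weapon sound class.
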MetaSoundSource /VanillaSoda/Assets/Audio/Pull/MSS_VanillaSoda_Grip_OneShotSingle
inputs [
 Sound {type=WaveAsset}
 UE.Source.OnPlay {type=Trigger}
]
interfaces [UE.OutputFormat.Mono:1.0, UE.Source.OneShot:1.0, UE.Source:1.1]
outputs [
 "UE.OutputFormat.Mono.Audio:0" {type=Audio}
 UE.Source.OneShot.OnFinished {type=Trigger}
]
pages [Default {
  input_defaults {Sound=/VanillaSoda/Assets/Audio/Pull/Wavs/VanillaSoda_StonePull_Grip_01, UE.Source.OnPlay=null}
  nodes [WavePlayer_0 = UE.Wave Player:Mono(Wave Asset=^Sound, Play=^UE.Source.OnPlay) {also {version [1,1]}}]
  links [{dst="@UE.Source.OneShot.OnFinished", src=WavePlayer_0.On Finished}, {dst="@UE.OutputFormat.Mono.Audio:0", src=WavePlayer_0.Out Mono}]
 }]
properties {
 assetImportData=&self:AssetImportData
 attenuationSettings=&/ATT_Log_Distance_8000
 maxDistance=8400
 sampleRate=48000
 soundClassObject=&Fort_WeaponMelee
}
